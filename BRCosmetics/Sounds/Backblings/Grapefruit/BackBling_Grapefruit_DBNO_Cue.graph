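# Sound cue for the Grapefruit back bling's DBNO (down but not out) sound.
# It plays a single wave through a modulator after a short fixed delay, with volume held fixed at the modulator and lowered on the cue.
# Close-range pet attenuation, the back bling sound class and a reactive back bling cosmetic property sheet govern its playback.
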SoundCue /BRCosmetics/Sounds/Backblings/Grapefruit/BackBling_Grapefruit_DBNO_Cue
root=Delay_0
nodes [
 {id=Delay_0, children [Modulator_0], properties {delayMax=0.2, delayMin=0.2}}
 {id=Modulator_0, children [WavePlayer_1], properties {volumeMax=1, volumeMin=1}}
 {id=WavePlayer_1, properties {soundWaveAssetPtr=&/BackBling_Grapefruit_DBNO_01}}
]
properties {
 attenuationSettings=&Fort_Pets_Close_Athena_Att
 audioPropertiesSheet=/BRCosmetics/Sounds/Settings/PropertySheets/Reactive/APS_Cosmetic_Reactive_Backbling
 maxDistance=1800.0001
 soundClassObject=/Game/Sounds/FortSoundClassesAndModes/Fort_Backblings
 volumeMultiplier=0.2
}
