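# Sound cue for the Grapefruit back bling's damaged sound, a reactive cosmetic effect.
# A single wave plays through a modulator that randomizes pitch and volume, after a short fixed delay.
# It uses close pet attenuation, the backbling sound class and a reactive backbling property sheet.
# Overridden concurrency limits instances per owner, blocking new plays at the cap, with a retrigger time.
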SoundCue /BRCosmetics/Sounds/Backblings/Grapefruit/BackBling_Grapefruit_Damaged_Cue
root=Delay_0
nodes [
 {id=Delay_0, children [Modulator_0], properties {delayMax=0.15, delayMin=0.15}}
 {id=Modulator_0, children [WavePlayer_1], properties {pitchMax=1.25, pitchMin=0.9, volumeMin=0.65}}
 {id=WavePlayer_1, properties {soundWaveAssetPtr=&/BackBling_Grapefruit_Damaged_01}}
]
properties {
 attenuationSettings=&Fort_Pets_Close_Athena_Att
 audioPropertiesSheet=/BRCosmetics/Sounds/Settings/PropertySheets/Reactive/APS_Cosmetic_Reactive_Backbling
 bOverrideConcurrency=true
 maxDistance=1800.0001
 soundClassObject=/Game/Sounds/FortSoundClassesAndModes/Fort_Backblings
 volumeMultiplier=0.35
 concurrencyOverrides {bLimitToOwner=true, maxCount=3, resolutionRule=PreventNew, retriggerTime=0.2}
}
